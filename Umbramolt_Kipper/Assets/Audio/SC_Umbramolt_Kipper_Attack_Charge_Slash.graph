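# Sound cue for the Umbramolt Kipper's charge slash attack, routed to the gadgets/items sound class with attack attenuation.
# Mixes a random charge swing with a quieter random dive-slash launch, each drawn from three equally weighted waves.
# Branches on islocal then isenemy play that mix at half volume when both are false, otherwise full.
# A root modulator adds slight random variation over the result.
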SoundCue /Umbramolt_Kipper/Assets/Audio/SC_Umbramolt_Kipper_Attack_Charge_Slash
root=Modulator_0
nodes [
 {id=Branch_2, children [Modulator_2,Branch_3,Modulator_2], properties {boolParameterName=islocal}}
 {id=Branch_3, children [Modulator_4,Modulator_3,Modulator_4], properties {boolParameterName=isenemy}}
 {id=Mixer_0, children [Random_0,Random_1], properties {inputVolume [0.7,0.3]}}
 {id=Modulator_0, children [Branch_2], properties {pitchMin=1}}
 {id=Modulator_2, children [Mixer_0], properties {pitchMax=1, pitchMin=1, volumeMax=1, volumeMin=1}}
 {id=Modulator_3, children [Mixer_0], properties {pitchMax=1, pitchMin=1, volumeMax=0.5, volumeMin=0.5}}
 {id=Modulator_4, children [Mixer_0], properties {pitchMax=1, pitchMin=1, volumeMax=1, volumeMin=1}}
 {id=Random_0, children [WavePlayer_1,WavePlayer_2,WavePlayer_3], properties {weights [1,1,1]}}
 {id=Random_1, children [WavePlayer_4,WavePlayer_5,WavePlayer_6], properties {weights [1,1,1]}}
 {id=WavePlayer_1, properties {soundWaveAssetPtr=&/Umbramolt_Kipper_Attack_Charge_Swing_03}}
 {id=WavePlayer_2, properties {soundWaveAssetPtr=&/Umbramolt_Kipper_Attack_Charge_Swing_02}}
 {id=WavePlayer_3, properties {soundWaveAssetPtr=&/Umbramolt_Kipper_Attack_Charge_Swing_01}}
 {id=WavePlayer_4, properties {soundWaveAssetPtr=&/Umbramolt_Kipper_Attack_Dive_Slash_Launch_01}}
 {id=WavePlayer_5, properties {soundWaveAssetPtr=&/Umbramolt_Kipper_Attack_Dive_Slash_Launch_02}}
 {id=WavePlayer_6, properties {soundWaveAssetPtr=&/Umbramolt_Kipper_Attack_Dive_Slash_Launch_03}}
]
properties {
 attenuationSettings=&/Att_Umbramolt_Kipper_Attacks
 bEnableBusSends=false
 maxDistance=5500
 soundClassObject=&Fort_Gadgets_Items
 volumeMultiplier=1.35
 busSends [{audioBus=null, minSendLevel=0, sendLevel=1, soundSourceBus=null}]
 preEffectBusSends [{soundSourceBus=null}]
}
