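# Sound cue for the start of the Umbramolt Kipper's charge attack, playing a single wave through a one-entry random node.
# Branches on the islocal and isenemy parameters halve the volume only when both are false, leaving local, enemy and unset cases at full volume.
# It uses the Kipper attack attenuation and gadgets/items sound class, with a low-level pre-effect send to the weapon sidechain bus.
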SoundCue /Umbramolt_Kipper/Assets/Audio/SC_Umbramolt_Kipper_Attack_Charge_Start
root=Branch_2
nodes [
 {id=Branch_2, children [Modulator_2,Branch_3,Modulator_2], properties {boolParameterName=islocal}}
 {id=Branch_3, children [Modulator_4,Modulator_3,Modulator_4], properties {boolParameterName=isenemy}}
 {id=Modulator_2, children [Random_0], properties {pitchMax=1, pitchMin=1, volumeMax=1, volumeMin=1}}
 {id=Modulator_3, children [Random_0], properties {pitchMax=1, pitchMin=1, volumeMax=0.5, volumeMin=0.5}}
 {id=Modulator_4, children [Random_0], properties {pitchMax=1, pitchMin=1, volumeMax=1, volumeMin=1}}
 {id=Random_0, children [WavePlayer_3], properties {weights [1]}}
 {id=WavePlayer_3, properties {soundWaveAssetPtr=&/Umbramolt_Kipper_Attack_Charge_Start_01}}
]
properties {
 attenuationSettings=&/Att_Umbramolt_Kipper_Attacks
 maxDistance=5500
 soundClassObject=&Fort_Gadgets_Items
 volumeMultiplier=1
 busSends [{audioBus=null, minSendLevel=0, sendLevel=1, soundSourceBus=null}]
 preEffectBusSends [{audioBus=&AB_Sidechain_FortWeapon_Vol, sendLevel=0.1, soundSourceBus=null}]
}
