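# Sound cue for the Umbramolt Kipper attack charge-up, randomly picking one of three equally weighted MidCharge waves.
# Nested branches on the islocal and isenemy parameters route that shared random pick through volume modulators, halving it for non-local, non-enemy sources and keeping full volume otherwise.
# An attack attenuation asset shapes falloff, and a concurrency override allows one instance per owner, stopping the oldest.
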SoundCue /Umbramolt_Kipper/Assets/Audio/SC_Umbramolt_Kipper_Attack_Chargeup
root=Branch_2
nodes [
 {id=Branch_2, children [Modulator_0,Branch_3,Modulator_0], properties {boolParameterName=islocal}}
 {id=Branch_3, children [Modulator_4,Modulator_3,Modulator_4], properties {boolParameterName=isenemy}}
 {id=Modulator_0, children [Modulator_2], properties {pitchMax=1, pitchMin=1, volumeMax=1, volumeMin=1}}
 {id=Modulator_2, children [Random_0]}
 {id=Modulator_3, children [Modulator_2], properties {pitchMax=1, pitchMin=1, volumeMax=0.5, volumeMin=0.5}}
 {id=Modulator_4, children [Modulator_2], properties {pitchMax=1, pitchMin=1, volumeMax=1, volumeMin=1}}
 {id=Random_0, children [WavePlayer_1,WavePlayer_2,WavePlayer_3], properties {weights [1,1,1]}}
 {id=WavePlayer_1, properties {soundWaveAssetPtr=&/Umbramolt_Kipper_MidCharge_01}}
 {id=WavePlayer_2, properties {soundWaveAssetPtr=&/Umbramolt_Kipper_MidCharge_02}}
 {id=WavePlayer_3, properties {soundWaveAssetPtr=&/Umbramolt_Kipper_MidCharge_03}}
]
properties {
 attenuationSettings=&/Att_Umbramolt_Kipper_Attacks
 bEnableBusSends=false
 bOverrideConcurrency=true
 maxDistance=5500
 soundClassObject=&Fort_Gadgets_Items
 volumeMultiplier=1
 busSends [{audioBus=null, minSendLevel=0, sendLevel=1, soundSourceBus=null}]
 concurrencyOverrides {bLimitToOwner=true, maxCount=1, resolutionRule=StopOldest}
 preEffectBusSends [{audioBus=&AB_Sidechain_FortWeapon_Vol, soundSourceBus=null}]
}
