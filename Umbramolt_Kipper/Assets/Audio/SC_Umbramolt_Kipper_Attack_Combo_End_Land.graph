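# Sound cue for the landing sound that ends the Umbramolt Kipper attack combo, picking one of three equally weighted waves through a shared modulator.
# Branches on islocal and isenemy halve the volume for non-local, non-enemy sources, while local, enemy and unset cases play at full modulator level.
# It uses the Kipper attack attenuation with overrides and player melee weapon concurrency; sidechain bus sends are configured but bus sends are disabled.
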SoundCue /Umbramolt_Kipper/Assets/Audio/SC_Umbramolt_Kipper_Attack_Combo_End_Land
root=Branch_2
nodes [
 {id=Branch_2, children [Modulator_2,Branch_3,Modulator_2], properties {boolParameterName=islocal}}
 {id=Branch_3, children [Modulator_4,Modulator_3,Modulator_4], properties {boolParameterName=isenemy}}
 {id=Modulator_0, children [Random_1]}
 {id=Modulator_2, children [Modulator_0], properties {pitchMax=1, pitchMin=1, volumeMax=1, volumeMin=1}}
 {id=Modulator_3, children [Modulator_0], properties {pitchMax=1, pitchMin=1, volumeMax=0.5, volumeMin=0.5}}
 {id=Modulator_4, children [Modulator_0], properties {pitchMax=1, pitchMin=1, volumeMax=1, volumeMin=1}}
 {id=Random_1, children [WavePlayer_3,WavePlayer_4,WavePlayer_5], properties {weights [1,1,1]}}
 {id=WavePlayer_3, properties {soundWaveAssetPtr=&/Umbramolt_Kipper_Attack_Combo_End_Land_01}}
 {id=WavePlayer_4, properties {soundWaveAssetPtr=&/Umbramolt_Kipper_Attack_Combo_End_Land_02}}
 {id=WavePlayer_5, properties {soundWaveAssetPtr=&/Umbramolt_Kipper_Attack_Combo_End_Land_03}}
]
properties {
 attenuationSettings=&/Att_Umbramolt_Kipper_Attacks
 bEnableBusSends=false
 maxDistance=5500
 soundClassObject=&Fort_Gadgets_Items
 volumeMultiplier=1.5
 attenuationOverrides {distanceAlgorithm=NaturalSound, falloffDistance=2000, attenuationShapeExtents {x=500}}
 busSends [{audioBus=&AB_Sidechain_FortWeapon_Vol, minSendLevel=0, sendLevel=1, soundSourceBus=null}]
 concurrencySet [&Player_Melee_Weapon_Conc]
 preEffectBusSends [{audioBus=&AB_Sidechain_FortWeapon_Vol, soundSourceBus=null}]
}
