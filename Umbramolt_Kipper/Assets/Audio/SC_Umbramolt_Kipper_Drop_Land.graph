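# Sound cue for the landing sound of the Umbramolt Kipper when dropped.
# A Random node picks one of three land waves with equal weight and feeds a Modulator that holds pitch fixed.
# The cue plays at reduced volume in the melee weapon sound class, using shared weapon-drop attenuation with bus sends disabled.
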SoundCue /Umbramolt_Kipper/Assets/Audio/SC_Umbramolt_Kipper_Drop_Land
root=Modulator_0
nodes [
 {id=Modulator_0, children [Random_0], properties {pitchMax=1, pitchMin=1}}
 {id=Random_0, children [WavePlayer_0,WavePlayer_1,WavePlayer_2], properties {weights [1,1,1]}}
 {id=WavePlayer_0, properties {soundWaveAssetPtr=&/Umbramolt_Kipper_Land_01}}
 {id=WavePlayer_1, properties {soundWaveAssetPtr=&/Umbramolt_Kipper_Land_02}}
 {id=WavePlayer_2, properties {soundWaveAssetPtr=&/Umbramolt_Kipper_Land_03}}
]
properties {
 attenuationSettings=&Fort_Weapon_Drop_Att
 bEnableBusSends=false
 maxDistance=3000
 soundClassObject=&Fort_WeaponMelee
 volumeMultiplier=0.4
}
