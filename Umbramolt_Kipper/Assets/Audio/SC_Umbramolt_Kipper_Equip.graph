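# Sound cue for the equip sound of the Umbramolt Kipper, classed as a melee weapon sound.
# It randomly picks one of three equally weighted wave variations and passes it through a modulator at the root.
# Playback is at reduced volume with a dedicated attenuation asset.
# A pre-effect send to the weapon volume sidechain bus is configured, but bus sends are disabled.
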SoundCue /Umbramolt_Kipper/Assets/Audio/SC_Umbramolt_Kipper_Equip
root=Modulator_0
nodes [
 {id=Modulator_0, children [Random_0]}
 {id=Random_0, children [WavePlayer_0,WavePlayer_1,WavePlayer_2], properties {weights [1,1,1]}}
 {id=WavePlayer_0, properties {soundWaveAssetPtr=&/Umbramolt_Kipper_Equip_01}}
 {id=WavePlayer_1, properties {soundWaveAssetPtr=&/Umbramolt_Kipper_Equip_02}}
 {id=WavePlayer_2, properties {soundWaveAssetPtr=&/Umbramolt_Kipper_Equip_03}}
]
properties {
 attenuationSettings=&/SA_Umbramolt_Kipper_Equip
 bEnableBusSends=false
 maxDistance=2800
 soundClassObject=&Fort_WeaponMelee
 volumeMultiplier=0.4
 busSends [{audioBus=null, minSendLevel=0, sendLevel=1, soundSourceBus=null}]
 preEffectBusSends [{audioBus=&AB_Sidechain_FortWeapon_Vol, soundSourceBus=null}]
}
