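# Sound cue for the normal impact sound of the Umbramolt Kipper, apparently a melee weapon, played with distance attenuation.
# A Random node picks one of three equally weighted impact waves and feeds a Modulator for pitch and volume variation.
# It plays at reduced volume in the gadgets/items sound class under player melee weapon concurrency.
# A weapon sidechain bus send is configured, but bus sends are disabled.
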SoundCue /Umbramolt_Kipper/Assets/Audio/SC_Umbramolt_Kipper_Impact_Normal_Att
root=Modulator_0
nodes [
 {id=Modulator_0, children [Random_0]}
 {id=Random_0, children [WavePlayer_1,WavePlayer_2,WavePlayer_3], properties {weights [1,1,1]}}
 {id=WavePlayer_1, properties {soundWaveAssetPtr=&/Umbramolt_Kipper_Impact_01}}
 {id=WavePlayer_2, properties {soundWaveAssetPtr=&/Umbramolt_Kipper_Impact_02}}
 {id=WavePlayer_3, properties {soundWaveAssetPtr=&/Umbramolt_Kipper_Impact_03}}
]
properties {
 attenuationSettings=&/AS_Umbramolt_Kipper_Impacts
 bEnableBusSends=false
 maxDistance=2500
 soundClassObject=&Fort_Gadgets_Items
 volumeMultiplier=0.4
 busSends [{audioBus=null, minSendLevel=0, sendLevel=1, soundSourceBus=null}]
 concurrencyOverrides {resolutionRule=PreventNew}
 concurrencySet [&Player_Melee_Weapon_Conc]
 preEffectBusSends [{audioBus=&AB_Sidechain_FortWeapon_Vol, soundSourceBus=null}]
}
