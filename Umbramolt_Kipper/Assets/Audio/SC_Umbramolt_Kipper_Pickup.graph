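# Sound cue for picking up the Umbramolt Kipper, classed as loot collection and using weapon pickup attenuation.
# A random Kipper-specific pickup wave, slightly delayed and boosted, is mixed over a random Katana pickup wave.
# The mix feeds a root modulator with pitch variation flattened.
# A concurrency set for pickup sounds is assigned, alongside single-instance override values.
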SoundCue /Umbramolt_Kipper/Assets/Audio/SC_Umbramolt_Kipper_Pickup
root=Modulator_0
nodes [
 {id=Delay_1, children [Random_0], properties {delayMax=0.1, delayMin=0.1}}
 {id=Mixer_1, children [Delay_1,Random_1], properties {inputVolume [2,1]}}
 {id=Modulator_0, children [Mixer_1], properties {pitchMax=1, pitchMin=1}}
 {id=Random_0, children [WavePlayer_0,WavePlayer_1,WavePlayer_3], properties {weights [1,1,1]}}
 {id=Random_1, children [WavePlayer_2,WavePlayer_4,WavePlayer_5], properties {weights [1,1,1]}}
 {id=WavePlayer_0, properties {soundWaveAssetPtr=&/Umbramolt_Kipper_Pickup_01}}
 {id=WavePlayer_1, properties {soundWaveAssetPtr=&/Umbramolt_Kipper_Pickup_02}}
 {id=WavePlayer_2, properties {soundWaveAssetPtr=&/Katana_Pickup_01}}
 {id=WavePlayer_3, properties {soundWaveAssetPtr=&/Umbramolt_Kipper_Pickup_03}}
 {id=WavePlayer_4, properties {soundWaveAssetPtr=&/Katana_Pickup_02}}
 {id=WavePlayer_5, properties {soundWaveAssetPtr=&/Katana_Pickup_03}}
]
properties {
 attenuationSettings=&Fort_Weapon_Pickup_Att
 bEnableBusSends=false
 maxDistance=3500
 soundClassObject=&Fort_Loot_Collect
 volumeMultiplier=0.4
 concurrencyOverrides {maxCount=1, resolutionRule=PreventNew}
 concurrencySet [&PickupMaximum_Oldest]
}
