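# Sound cue for the Umbramolt Kipper pickup, playing a music stinger when the item is collected.
# A single wave player feeds the output directly at reduced volume, with no randomization or mixing.
# It borrows shared weapon-pickup attenuation and the loot-collect sound class, with bus sends disabled.
# The shared PickupMaximum_Oldest concurrency set is assigned to govern overlapping pickup sounds.
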SoundCue /Umbramolt_Kipper/Assets/Audio/SC_Umbramolt_Kipper_Pickup_MX
root=WavePlayer_0
nodes [{id=WavePlayer_0, properties {soundWaveAssetPtr=&/MX_S39_UmbraMolt_Pickup_STNG}}]
properties {
 attenuationSettings=&Fort_Weapon_Pickup_Att
 bEnableBusSends=false
 maxDistance=3500
 soundClassObject=&Fort_Loot_Collect
 volumeMultiplier=0.7
 assetUserData [&self:AudioAssetUserData_0]
 concurrencyOverrides {maxCount=1, resolutionRule=PreventNew}
 concurrencySet [&PickupMaximum_Oldest]
}
