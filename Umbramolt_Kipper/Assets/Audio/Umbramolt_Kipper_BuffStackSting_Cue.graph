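# Sound cue for the Umbramolt Kipper buff stack sting, a one-shot whose exact trigger is not shown here.
# A single wave player feeds the output directly, pointing at a wave whose name marks it as a do-not-use movie placeholder.
# It borrows Athena emote attenuation and foley concurrency, uses the gadget items sound class, and lists a send to the music audio-analysis submix.
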SoundCue /Umbramolt_Kipper/Assets/Audio/Umbramolt_Kipper_BuffStackSting_Cue
root=WavePlayer_2
nodes [{id=WavePlayer_2, properties {soundWaveAssetPtr=&/MOVIEPLACEHOLDERDONOTUSE_Umbramolt_Kipper_BuffStackSting}}]
properties {
 attenuationSettings=&Fort_Emotes_Athena_Att
 bEnableBusSends=false
 maxDistance=1600
 soundClassObject=&Fort_Gadgets_Items
 concurrencyOverrides {maxCount=3, resolutionRule=StopOldest}
 concurrencySet [/Game/Sounds/Concurrency_Rules/Athena_EmoteFoley_Conc]
 soundSubmixSends [{minSendLevel=0, soundSubmix=&Fortnite_AudioAnalysis_Music}]
}
