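# Two-sided translucent lit material for Niagara mesh particles, apparently a flat blade slash trail.
# A dynamic parameter slides a clamped slash ramp texture, alongside panned noise, Fresnel, UV edge masks, and camera and depth fades for opacity work.
# An inverse-proportional decay curve shapes particle-colored emissive.
# Dynamic-parameter WPO offsets vertices along instance-local radial directions, flipping sign across V to set width, and a quality switch zeroes refraction on low-end platforms.
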Material /Umbramolt_Kipper/Assets/FX/M_FlatBladeTrail
graph {
 properties {
  bAutomaticallySetUsageInEditor=false
  bUsedWithNiagaraMeshParticles=true
  blendMode=BLEND_Translucent
  refractionMethod=RM_2DOffset
  twoSided=true
  usedShadingModels=Default Lit
 }
 nodes [
  ChannelMaskParameter() {group=Opacity, parameterName=OpacityChannel}
  Constant() {r=1}
  Constant() {r=1}
  n1 = TextureCoordinate()
  LinearInterpolate(
   A=@n1
   B=MaterialFunctionCall(XY=@n1) {materialFunction=&Swizzle}.YX
   Alpha=ScalarParameter() {defaultValue=1, parameterName=Main_UV_Swizzle}
  )
  n2 = TextureCoordinate()
  LinearInterpolate(
   A=@n2
   B=MaterialFunctionCall(XY=@n2) {materialFunction=&Swizzle}.YX
   Alpha=ScalarParameter() {defaultValue=1, parameterName=UV1_Swizzle}
  )
  n3 = ScalarParameter() {defaultValue=0.1, group=InverseProportionalDecay, parameterName=K}
  n4 = NamedRerouteUsage()
  n5 = Subtract(A=Divide(A=@n3, B=OneMinus(None=@n4)), B=@n3)
  LinearInterpolate(
   A=VectorParameter() {group=Emissive, parameterName=EmissiveColorTintA, defaultValue=<rgba 0.0029 0.0218 0.0469 1>}.RGB
   B=VectorParameter() {group=Emissive, parameterName=EmissiveColorTintB, defaultValue=<rgba 38.7791 60.8518 100 1>}.RGB
   Alpha=@n5
  )
  n6 = NamedRerouteUsage()
  n7 = NamedRerouteUsage()
  LinearInterpolate(A=Multiply(A=@n6, B=@n7), B=@n7) {constAlpha=0.05}
  MaterialFunctionCall(Emissive=NamedRerouteUsage()) {materialFunction=&MF_FX_EmissiveAdjustments}
  n8 = Constant()
  n9 = Constant3Vector() {constant=<rgba 1 0 0 0>}
  n10 = Multiply(
   A=Multiply(A=Transform(None=@n9) {transformType=TRANSFORM_View}, B=NamedRerouteUsage())
   B=ScalarParameter() {defaultValue=0.06, parameterName=Refraction}
  )
  MaterialFunctionCall!(
   Low _ PC Low and Med=@n8
   "High _ XB1, PS4"=@n10
   "Epic _ PC High, XSX, PS5"=@n10
   --------
   "Mobile Low _ PC Perf, Android"=@n8
   Mobile Med _ Switch=@n8
  ) {materialFunction=/Game/Global/MaterialFunctions/TechArt/MF_Fortnite_QualitySwitch_Vector3}
  n11 = VectorParameter() {parameterName=Main_UV, defaultValue=<rgba 1 1 -0.3067 0>}
  NamedRerouteDeclaration(None=MaterialFunctionCall!(
    ClampTexture
    UV=@n1
    Tex
    InRGB=@n11.RGB
    InA=@n11.RGB
    Time=DynamicParameter() {defaultValue=<rgba 1 1 1 250>, paramNames [Time, Distort, Dissolve, ""]}.Time
    TimeMult
   ) {materialFunction=/Game/Effects/Niagara/MasterMaterials/Functions/MF_FX_VectorDrive_TexPanner}.PannerUV) {name=MainUV, nodeColor=<rgba 0 0.4049 1 1>}
  n12 = Subtract(A=CameraPositionWS(), B=WorldPosition().XYZ)
  n13 = Length(None=@n12) {is=class MaterialExpressionLength}
  n14 = DotProduct(A=Normalize(VectorInput=@n12), B=Transform(None=Constant3Vector() {constant=<rgba 0 0 -1 0>}) {transformSourceType=TRANSFORMSOURCE_View})
  n15 = ScalarParameter() {parameterName=Camfade_Offset}
  n16 = ScalarParameter() {defaultValue=300, parameterName=Camfade_Lemgth}
  n17 = Divide(A=Subtract(A=Multiply(A=@n13, B=@n14), B=@n15), B=@n16)
  n18 = Divide(A=Subtract(A=PixelDepth().R, B=@n15), B=@n16)
  n19 = StaticBool() {value=true}
  NamedRerouteDeclaration(None=Saturate(None=StaticSwitch(True=@n17, False=@n18, Value=@n19))) {name=CamFade, nodeColor=<rgba 1 0 0.9613 1>}
  n20 = CameraPositionWS()
  n21 = WorldPosition()
  n22 = DotProduct(A=VertexNormalWS(), B=Normalize(VectorInput=Subtract(A=@n20, B=@n21.XYZ)))
  n23 = Abs(None=@n22)
  n24 = Power(
   Base=LinearInterpolate(A=OneMinus(None=@n23), B=@n23, Alpha=ScalarParameter() {defaultValue=1, parameterName="Fresnel_Invert (0or1)"})
   Exp=ScalarParameter() {defaultValue=0.5, parameterName=Fresnel_Power}
  )
  n25 = ScalarParameter() {defaultValue=0.05, parameterName=Fresnel_Offset}
  NamedRerouteDeclaration(None=Multiply(A=Subtract(A=@n24, B=@n25), B=Add(A=@n25))) {name=Fresnel, nodeColor=<rgba 1 0 0.9613 1>}
  n26 = MaterialFunctionCall!(
   InAlpha=NamedRerouteUsage()
   OpacityBasedFade_DistanceA=ScalarParameter() {defaultValue=50, group=Opacity, parameterName=OpacityBasedFade_DistanceA}
   OpacityBasedFade_DistanceB=ScalarParameter() {defaultValue=5, group=Opacity, parameterName=OpacityBasedFade_DistanceB}
  ) {materialFunction=/Game/Effects/Niagara/MasterMaterials/Functions/MF_FX_OpacityBasedDepthFade}
  NamedRerouteDeclaration(None=@n26.FadeValue) {name=DepthFadeValue, nodeColor=<rgba 0 1 0.5312 1>}
  n27 = VectorParameter() {parameterName=UV1, defaultValue=<rgba 1 1 -0.3067 0>}
  n28 = MaterialFunctionCall(UV=@n2, InRGB=@n27.RGB, InA=@n27.RGB, Time=DynamicParameter() {defaultValue=<rgba 1 1 1 250>, paramNames [Time, Distort, Dissolve, ""]}.Time) {materialFunction=&MF_FX_VectorDrive_TexPanner}
  NamedRerouteDeclaration(None=TextureSampleParameter2D(UVs=NamedRerouteDeclaration(None=@n28.PannerUV) {name=UV_1, nodeColor=<rgba 0 0.4049 1 1>}) {parameterName=NoiseTexture, texture=/Game/Effects/Fort_Effects/Textures/Tile/T_Tile_WaveFlame}.G) {name=MV_SampledTexture_1, nodeColor=<rgba 0.0153 1 0 1>}
  NamedRerouteDeclaration(None=Multiply(A=ParticleColor().RGB, B=@n5)) {name=Emissive, nodeColor=<rgba 0 0 0 1>}
  n29 = TextureCoordinate()
  n30 = TextureCoordinate() {coordinateIndex=1}
  n31 = ScalarParameter() {parameterName=CubeMap_Switch}
  n32 = ScalarParameter() {defaultValue=1.5, parameterName=CubeMap_Offset}
  n33 = Subtract(A=Multiply(A=LinearInterpolate(A=@n29, B=@n30, Alpha=@n31), B=Multiply(A=@n32) {constB=2}) {constB=4}, B=@n32) {constB=2}
  n34 = Multiply(A=@n33, B=@n33) {constB=2}
  n35 = Saturate(None=OneMinus(None=Abs(None=@n34)))
  n36 = MaterialFunctionCall(Float2=Multiply(A=@n35, B=@n35)) {materialFunction=&BreakOutFloat2Components}
  NamedRerouteDeclaration(None=MaterialFunctionCall(A=@n36.R, B=Multiply(A=@n36.R, B=@n36.R), C=@n36.R, Alpha=ScalarParameter() {parameterName=CubeMap_Select}) {materialFunction=&Lerp_3Color}.Result) {name=CubeMap, nodeColor=<rgba 1 1 1 1>}
  n37 = ParticleColor()
  n38 = Multiply(A=NamedRerouteUsage(), B=NamedRerouteUsage())
  n39 = NamedRerouteUsage()
  NamedRerouteDeclaration(None=Multiply(A=Multiply(A=@n37.A, B=@n26.FinalAlpha), B=Multiply(A=@n38, B=@n39))) {name=Opacity, nodeColor=<rgba 0 0 0 1>}
  n40 = DynamicParameter() {parameterIndex=1, defaultValue=<rgba 50 0.5 0.5 0.5>, paramNames [SlashRampData_Animation_G, WPO_WidthOffset_A, WPO_WidthOffset_B, ""]}
  n41 = TextureCoordinate()
  n42 = Multiply(A=ComponentMask(None=@n41) {r=true}, B=ScalarParameter() {defaultValue=0.25, parameterName=SlashRampData_Threshold})
  n43 = TextureSampleParameter2D(UVs=AppendVector(A=Add(A=@n40.SlashRampData_Animation_G, B=@n42), B=ComponentMask(None=@n41) {g=true})) {
   parameterName=SlashRampData
   samplerSource=SSM_Clamp_WorldGroupSettings
   samplerType=SAMPLERTYPE_LinearColor
   texture=&/T_BladeRampData
  }
  NamedRerouteDeclaration(None=@n43.G) {name=SlashRamp, nodeColor=<rgba 1 1 1 1>}
  NamedRerouteDeclaration(None=LinearInterpolate(A=@n6, B=@n7)) {name=Alpha, nodeColor=<rgba 1 0.1875 0 1>}
  n44 = Normalize(VectorInput=TransformPosition(None=WorldPosition() {worldPositionShaderOffset=WPT_ExcludeAllShaderOffsets}.XYZ) {transformSourceType=TRANSFORMPOSSOURCE_World, transformType=TRANSFORMPOSSOURCE_Instance})
  n45 = DynamicParameter() {parameterIndex=1, defaultValue=<rgba 50 0.5 0.5 0.5>, paramNames [SlashRampData_Animation_G, WPO_WidthOffset_A, WPO_WidthOffset_B, ""]}
  n46 = ComponentMask(None=TextureCoordinate()) {g=true}
  n47 = LinearInterpolate(A=@n45.WPO_WidthOffset_B, B=@n45.WPO_WidthOffset_B, Alpha=@n46)
  n48 = LinearInterpolate(Alpha=@n46) {constA=1, constB=-1}
  NamedRerouteDeclaration(None=Multiply(A=Transform(None=@n44) {transformSourceType=TRANSFORMSOURCE_Instance}, B=Multiply(A=@n47, B=@n48))) {name=WPO, nodeColor=<rgba 1 1 1 1>}
  NamedRerouteUsage()
  NamedRerouteUsage()
  NamedRerouteUsage()
  Reroute(None=@n22)
  VectorParameter() {parameterName=BaseColor, defaultValue=<rgba 0.0653 0.0951 0.151 1>}
  VertexTangentWS()
 ]
}
