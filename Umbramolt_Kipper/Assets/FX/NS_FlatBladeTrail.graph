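# Niagara system for a flat blade trail effect, with a single Slash emitter that renders bright HDR white mesh particles.
# SpawnRate emits particles continuously, each oriented at spawn, given a short fixed lifetime and a random material value.
# Over each particle's life, one curve swells the mesh quickly then eases out, while another drops a dynamic material parameter steeply toward zero, likely fading or eroding the slash.
# It exposes no user parameters.
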NiagaraSystem /Umbramolt_Kipper/Assets/FX/NS_FlatBladeTrail
dependencies {usedDataInterfaces [&NiagaraDataInterfaceCurve], usedDynamicInputs [], usedModules [&SystemState]}
emitters {Slash {
  emitterUpdateScript [
   &EmitterState()
   &SpawnRate(SpawnRate=10, Use Spawn Probability=false)
  ]
  particleSpawnScript [
   &InitializeParticle(
    Color=<rgba 10 10 10 5>
    Lifetime=2
    Mesh Scale Mode=::NewEnumerator3
    Mesh Uniform Scale=2
    UsePositionOffset=false
    Write Material Random=true
   )
   &InitialMeshOrientation(Use Rotation Vector=false)
  ]
  particleUpdateScript [
   &ParticleState()
   &ScaleMeshSize(Scale Factor=&VectorFromFloat(Value=&FloatFromCurve(FloatCurve=DataInterface() {propertyValues {Curve {
         defaultValue=1
         keys [
          {arriveTangent=2.8252, interpMode=RCIM_Cubic, leaveTangent=2.8252, tangentMode=RCTM_User, time=0, value=1}
          {arriveTangent=0.2984, interpMode=RCIM_Cubic, leaveTangent=0.2984, time=0.0462, value=1.1936}
          <at 1 1.2984 RCIM_Cubic>
         ]
        }}})))
   &DynamicMaterialParameters(
    Index 1 Param 2=&FloatFromCurve(FloatCurve=DataInterface() {propertyValues {Curve {
        defaultValue=1
        keys [
         {arriveTangent=-4.2978, interpMode=RCIM_Cubic, leaveTangent=-4.2978, tangentMode=RCTM_User, time=0, value=1}
         {arriveTangent=-1, interpMode=RCIM_Cubic, leaveTangent=-1, time=0.1054, value=0.3861}
         <at 1 0 RCIM_Cubic>
        ]
       }}})
    Module.Index0Float0Write=true
    Module.Index0Float1Write=true
    Module.Index0Float2Write=true
    Module.Index0Float3Write=true
    Module.Index1Float0Write=false
    Module.Index1Float1Write=true
    Module.Index1Float2Write=false
    Module.Index1Float3Write=false
    Write Parameter Index 1=::NewEnumerator1
   )
  ]
  rendererClasses [&NiagaraMeshRendererProperties]
  renderers [{rendererClass=&NiagaraMeshRendererProperties, rendererIndex=0}]
 }}
summary {
 systemName=NS_FlatBladeTrail
 emitters [{emitterName=Slash, rendererClasses [&NiagaraMeshRendererProperties]}]
 userVariables []
}
