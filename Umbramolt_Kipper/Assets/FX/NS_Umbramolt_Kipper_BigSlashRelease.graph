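# Niagara effect for Kipper's big slash release, built from three pale-blue mesh emitters sized by the owner's X scale.
# Slash bursts two oppositely tilted meshes that snap from an X stretch and flash with flickering brightness.
# SlashTrail spawns from Slash particles, inherits their orientation and spins to a stop, while Slash_L2 independently emits thin, randomly rotated layers.
# Curve-driven dynamic material parameters with random offsets animate all three.
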NiagaraSystem /Umbramolt_Kipper/Assets/FX/NS_Umbramolt_Kipper_BigSlashRelease
dependencies {
 usedDataInterfaces [
  &NiagaraDataInterfaceArrayFloat3
  &NiagaraDataInterfaceCurve
  &NiagaraDataInterfaceColorCurve
  &NiagaraDataInterfaceParticleRead
 ]
 usedDynamicInputs []
 usedModules [&SystemState]
}
emitters {
 Slash {
  emitterUpdateScript [
   &EmitterState()
   &SpawnBurst_Instantaneous(Spawn Count=2, Use Spawn Probability=false, UseLoopCountLimit=false)
  ]
  particleSpawnScript [
   &InitializeParticle(
    Color=<rgba 3.5417 6.4695 10 2>
    Lifetime=1.2
    Mesh Scale=&Multiply_VectorByFloat(
     Vector=<1 1 1>
     Float=&Multiply_Float(A=2.5, B=&MakeFloatFromVector(Vector=$Engine.Owner.Scale, Channel=::NewEnumerator0))
    )
    Mesh Scale Mode=::NewEnumerator4
    UsePositionOffset=false
    Write Material Random=true
   )
   &InitialMeshOrientation(
    Mesh Orientation Mode=::NewEnumerator1
    Rotation=&Add_Vector(
     A=&SelectVectorFromArray(
      Direct Array Index=&ReturnExecIndex()
      Direct Array Mode=::NewEnumerator1
      Vector Selection Array=DataInterface() {propertyValues {FloatData [<0.05 0 0>, <-0.05 0 0>]}}
     )
     B=&RandomRangeVector(Minimum=<-0.02 -0.02 -0.05>, Maximum=<0.02 0.02 0.05>)
    )
    Use Rotation Vector=true
   )
  ]
  particleUpdateScript [
   &ParticleState()
   &ScaleMeshSize(Scale Factor=&MakeVector(
     X=&FloatFromCurve(FloatCurve=DataInterface() {propertyValues {Curve {keys [
          {arriveTangent=0.1538, interpMode=RCIM_Cubic, leaveTangent=0.1538, tangentMode=RCTM_User, time=0, value=5}
          {arriveTangent=-0.8755, interpMode=RCIM_Cubic, leaveTangent=-0.8755, tangentMode=RCTM_User, time=0.0988, value=1.1951}
          <at 1 1 RCIM_Cubic>
         ]}}})
     Y=1
     Z=1
    )) {module=ScaleMeshSize001}
   &ScaleMeshSize(Scale Factor=&VectorFromFloat(Value=&FloatFromCurve(FloatCurve=DataInterface() {propertyValues {Curve {
         defaultValue=0.9935
         keys [
          {arriveTangent=0.5558, interpMode=RCIM_Cubic, leaveTangent=0.5558, tangentMode=RCTM_User, time=0, value=1}
          {
           arriveTangent=0.419
           arriveTangentWeight=0.0667
           interpMode=RCIM_Cubic
           leaveTangent=0.419
           leaveTangentWeight=0.267
           tangentMode=RCTM_User
           time=0.2
           value=1.1
          }
          {interpMode=RCIM_Cubic, leaveTangent=0.0009, time=1, value=1.2}
         ]
        }}})))
   &DynamicMaterialParameters(
    Index 0 Param 1=&Add_Float(
     A=&RandomRangeFloat(Minimum=0, Maximum=999)
     B=&FloatFromCurve(FloatCurve=DataInterface() {propertyValues {Curve {
         defaultValue=1
         keys [
          {arriveTangent=4.1785, interpMode=RCIM_Cubic, leaveTangent=4.1785, tangentMode=RCTM_User, time=0, value=0}
          {arriveTangent=1.1179, interpMode=RCIM_Cubic, leaveTangent=1.1179, tangentMode=RCTM_User, time=0.2042, value=1.0048}
          {arriveTangent=0.0224, interpMode=RCIM_Cubic, leaveTangent=0.0224, tangentMode=RCTM_User, time=1, value=1}
         ]
        }}})
    )
    Index 0 Param 2=1
    Index 1 Param 1=&FloatFromCurve(FloatCurve=DataInterface() {propertyValues {Curve {
        defaultValue=1
        keys [
         {arriveTangent=-1.7274, interpMode=RCIM_Cubic, leaveTangent=-1.7274, tangentMode=RCTM_User, time=0.0041, value=0.8988}
         {arriveTangent=-0.424, interpMode=RCIM_Cubic, leaveTangent=-0.424, tangentMode=RCTM_User, time=0.5037, value=0.6519}
         <at 1 0 RCIM_Cubic>
        ]
       }}})
    Index 1 Param 2=50
    Module.Index0Float0Write=true
    Module.Index0Float1Write=true
    Module.Index0Float2Write=true
    Module.Index0Float3Write=true
    Module.Index1Float0Write=true
    Module.Index1Float1Write=true
    Module.Index1Float2Write=false
    Module.Index1Float3Write=false
    Write Parameter Index 1=::NewEnumerator1
   )
   &ScaleColor(
    Scale Alpha=&FloatFromCurve(FloatCurve=DataInterface() {propertyValues {Curve {
        defaultValue=1
        keys [
         {interpMode=RCIM_Cubic, tangentMode=RCTM_User, time=0, value=0}
         {interpMode=RCIM_Cubic, tangentMode=RCTM_User, time=0.1, value=1}
         {tangentMode=RCTM_User, time=1, value=0}
        ]
       }}})
    Scale RGB=&VectorFromFloat(Value=&FloatFromCurve(
      FloatCurve=DataInterface() {propertyValues {Curve {keys [<at 0.1431 100>,<at 0.2916 0.9>,<at 1 0>]}}}
      Scale Curve=&Waveform("Waveform [1]"=::NewEnumerator8, "[1] Frequency"=20, "[1] Amplitude Min/Max"=<-0.5 1>)
     ))
    ScaleA=true
    ScaleRGB=true
   )
   &CameraOffset(Camera Offset Amount=100)
  ]
  rendererClasses [&NiagaraMeshRendererProperties]
  renderers [{rendererClass=&NiagaraMeshRendererProperties, rendererIndex=0}]
 }
 SlashTrail {
  emitterUpdateScript [
   &EmitterState()
   &SpawnParticlesFromOtherEmitter(
    Attribute Reader=DataInterface() {propertyValues {EmitterBinding {bindingMode=Other, emitterName=Slash}}}
    Spawn Rate=8
    Use Maximum Age=false
    Use Minimum Age=false
   )
  ]
  particleSpawnScript [
   &InitializeParticle(
    Color=<rgba 0.1771 0.3235 0.5 0.5>
    Lifetime=2
    Mesh Scale=&Multiply_VectorByFloat(
     Vector=<0.5 1 1>
     Float=&Multiply_Float(A=3, B=&MakeFloatFromVector(Vector=$Engine.Owner.Scale, Channel=::NewEnumerator0))
    )
    Mesh Scale Mode=::NewEnumerator4
    UsePositionOffset=false
    Write Material Random=true
   )
   &SampleParticlesFromOtherEmitter(Mesh Orientation Sampling:ENiagara_DirectReadSamplingMode=::NewEnumerator0)
  ]
  particleUpdateScript [
   &ParticleState()
   &UpdateMeshOrientation(
    Rotation Rate=&FloatFromCurve(FloatCurve=DataInterface() {propertyValues {Curve {keys [
         {arriveTangent=-2.9722, interpMode=RCIM_Cubic, leaveTangent=-2.9722, tangentMode=RCTM_User, time=0, value=1}
         {arriveTangent=-1, interpMode=RCIM_Cubic, leaveTangent=-1, time=0.1688, value=0.2876}
         <at 1 0 RCIM_Cubic>
        ]}}})
    Rotation Vector=&RandomRangeVector(Minimum=<0 0 -2>, Maximum=<0 0 2>)
   )
   &ScaleMeshSize(Scale Factor=&VectorFromFloat(Value=&FloatFromCurve(FloatCurve=DataInterface() {propertyValues {Curve {
         defaultValue=0
         keys [
          {arriveTangent=1.2462, interpMode=RCIM_Cubic, leaveTangent=1.2462, leaveTangentWeight=0.1127, tangentMode=RCTM_User, time=0, value=0.8817}
          {arriveTangent=0.3334, interpMode=RCIM_Cubic, leaveTangent=0.3334, tangentMode=RCTM_User, time=0.2, value=1.0692}
          {arriveTangent=0.0067, interpMode=RCIM_Cubic, leaveTangent=0.0067, tangentMode=RCTM_User, time=1, value=1.1192}
         ]
        }}})))
   &DynamicMaterialParameters(
    Index 0 Param 1=&Add_Float(
     A=&RandomRangeFloat(Minimum=0, Maximum=999)
     B=&FloatFromCurve(FloatCurve=DataInterface() {propertyValues {Curve {
         defaultValue=1
         keys [
          {arriveTangent=4.1785, interpMode=RCIM_Cubic, leaveTangent=4.1785, tangentMode=RCTM_User, time=0, value=0}
          {arriveTangent=1.1179, interpMode=RCIM_Cubic, leaveTangent=1.1179, tangentMode=RCTM_User, time=0.2042, value=1.0048}
          {arriveTangent=0.0224, interpMode=RCIM_Cubic, leaveTangent=0.0224, tangentMode=RCTM_User, time=1, value=1}
         ]
        }}})
    )
    Index 0 Param 2=1
    Index 1 Param 1=&FloatFromCurve(FloatCurve=DataInterface() {propertyValues {Curve {
        defaultValue=1
        keys [{arriveTangent=-1.3626, interpMode=RCIM_Cubic, leaveTangent=-1.3626, tangentMode=RCTM_User, time=0.0041, value=0.9193}, <at 1 0 RCIM_Cubic>]
       }}})
    Index 1 Param 2=&FloatFromCurve(
     FloatCurve=DataInterface() {propertyValues {Curve {
        defaultValue=-1.0299
        keys [
         {arriveTangent=9.0725, interpMode=RCIM_Cubic, leaveTangent=9.0725, tangentMode=RCTM_User, time=0, value=0}
         {arriveTangent=-2.5249, interpMode=RCIM_Cubic, leaveTangent=-2.5249, tangentMode=RCTM_User, time=0.1408, value=0.9847}
         {arriveTangent=-0.0405, interpMode=RCIM_Cubic, leaveTangent=-0.0405, tangentMode=RCTM_User, time=1, value=0}
        ]
       }}}
     Scale Curve=10
    )
    Module.Index0Float0Write=true
    Module.Index0Float1Write=true
    Module.Index0Float2Write=true
    Module.Index0Float3Write=true
    Module.Index1Float0Write=true
    Module.Index1Float1Write=true
    Module.Index1Float2Write=true
    Module.Index1Float3Write=false
    Write Parameter Index 1=::NewEnumerator1
   )
   &ScaleColor(
    Scale Alpha=&FloatFromCurve(FloatCurve=DataInterface() {propertyValues {Curve {
        defaultValue=1
        keys [
         {interpMode=RCIM_Cubic, tangentMode=RCTM_User, time=0, value=0}
         {interpMode=RCIM_Cubic, tangentMode=RCTM_User, time=0.1, value=1}
         {tangentMode=RCTM_User, time=1, value=0}
        ]
       }}})
    Scale RGB=&VectorFromFloat(Value=&FloatFromCurve(FloatCurve=DataInterface() {propertyValues {Curve {keys [<at 0.1997 100>,<at 0.2776 0.9>,<at 1 0>]}}}))
    ScaleA=true
    ScaleRGB=true
   )
   &CameraOffset(Camera Offset Amount=50)
   &SolveForcesAndVelocity(Clamp Velocity=false, Limit Acceleration=false)
  ]
  rendererClasses [&NiagaraMeshRendererProperties]
  renderers [{rendererClass=&NiagaraMeshRendererProperties, rendererIndex=0}]
 }
 Slash_L2 {
  emitterUpdateScript [
   &EmitterState(Life Cycle Mode=::NewEnumerator1, Loop Behavior=::NewEnumerator1, Loop Duration=1, UseLoopDelay=false)
   &SpawnRate(SpawnRate=10, Use Spawn Probability=false)
  ]
  particleSpawnScript [
   &InitializeParticle(
    Color=<rgba 3.5417 6.4695 10 0.5>
    Lifetime=&RandomRangeFloat(Minimum=1, Maximum=2)
    Mesh Scale=&Multiply_VectorByFloat(
     Vector=<0.1 1 1>
     Float=&Multiply_Float(
      A=&RandomRangeFloat(Minimum=2.5, Maximum=3)
      B=&MakeFloatFromVector(Vector=$Engine.Owner.Scale, Channel=::NewEnumerator0)
     )
    )
    Mesh Scale Mode=::NewEnumerator4
    UsePositionOffset=false
    Write Material Random=true
   )
   &InitialMeshOrientation(
    Mesh Orientation Mode=::NewEnumerator1
    Rotation=&RandomRangeVector(Minimum=<-0.05 -0.05 -0.3>, Maximum=<0.05 0.05 0.3>)
    Use Rotation Vector=true
   )
  ]
  particleUpdateScript [
   &ParticleState()
   &ScaleMeshSize(Scale Factor=&VectorFromFloat(Value=&FloatFromCurve(FloatCurve=DataInterface() {propertyValues {Curve {
         defaultValue=1
         keys [
          {arriveTangent=2.8252, interpMode=RCIM_Cubic, leaveTangent=2.8252, tangentMode=RCTM_User, time=0, value=1}
          {arriveTangent=0.1, interpMode=RCIM_Cubic, leaveTangent=0.1, time=0.0462, value=1.05}
          <at 1 1.1 RCIM_Cubic>
         ]
        }}})))
   &DynamicMaterialParameters(
    Index 0 Param 1=&Add_Float(
     A=&RandomRangeFloat(Minimum=0, Maximum=999)
     B=&FloatFromCurve(FloatCurve=DataInterface() {propertyValues {Curve {
         defaultValue=1
         keys [
          {arriveTangent=4.1785, interpMode=RCIM_Cubic, leaveTangent=4.1785, tangentMode=RCTM_User, time=0, value=0}
          {arriveTangent=1.1179, interpMode=RCIM_Cubic, leaveTangent=1.1179, tangentMode=RCTM_User, time=0.2042, value=1.0048}
          {arriveTangent=0.0224, interpMode=RCIM_Cubic, leaveTangent=0.0224, tangentMode=RCTM_User, time=1, value=1}
         ]
        }}})
    )
    Index 1 Param 1=&FloatFromCurve(FloatCurve=DataInterface() {propertyValues {Curve {
        defaultValue=1
        keys [
         {arriveTangent=-5.5864, interpMode=RCIM_Cubic, leaveTangent=-5.5864, tangentMode=RCTM_User, time=0.0041, value=1.0586}
         {arriveTangent=-1.0664, interpMode=RCIM_Cubic, leaveTangent=-1.0664, time=0.2474, value=0.846}
         <at 1 -0.0035 RCIM_Cubic>
        ]
       }}})
    Module.Index0Float0Write=true
    Module.Index0Float1Write=true
    Module.Index0Float2Write=true
    Module.Index0Float3Write=true
    Module.Index1Float0Write=true
    Module.Index1Float1Write=false
    Module.Index1Float2Write=false
    Module.Index1Float3Write=false
    Write Parameter Index 1=::NewEnumerator1
   )
  ]
  rendererClasses [&NiagaraMeshRendererProperties]
  renderers [{rendererClass=&NiagaraMeshRendererProperties, rendererIndex=0}]
 }
}
summary {
 systemName=NS_Umbramolt_Kipper_BigSlashRelease
 emitters [
  {emitterName=Slash, rendererClasses [&NiagaraMeshRendererProperties]}
  {emitterName=SlashTrail, rendererClasses [&NiagaraMeshRendererProperties]}
  {emitterName=Slash_L2, rendererClasses [&NiagaraMeshRendererProperties]}
 ]
 userVariables []
}
