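# Niagara charge effect for Umbramolt Kipper: a short, bright blue-white sprite flare that flickers and rises.
# One emitter bursts four sprites, using execution index to pick size and rotation from arrays, forming crossed thin streaks, a square sprite and a very wide flare.
# Curves drive easing rotation and upward offset, while a dip-swell-collapse size curve multiplied by a high-frequency waveform adds the flicker.
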NiagaraSystem /Umbramolt_Kipper/Assets/FX/NS_Umbramolt_Kipper_Charge
dependencies {
 usedDataInterfaces [&NiagaraDataInterfaceArrayFloat2, &NiagaraDataInterfaceArrayFloat, &NiagaraDataInterfaceCurve]
 usedDynamicInputs []
 usedModules [&SystemState]
}
emitters {Minimal {
  emitterUpdateScript [
   &EmitterState()
   &SpawnBurst_Instantaneous(Spawn Count=4, Use Spawn Probability=false, UseLoopCountLimit=false)
  ]
  particleSpawnScript [&InitializeParticle(
    Color=<rgba 4.561 5.806 10 0.4522>
    Lifetime=0.65
    Sprite Rotation Angle=&SelectFloatFromArray(
     Direct Array Index=&ReturnExecIndex()
     Float Selection Array=DataInterface() {propertyValues {FloatData [45,45,0]}}
    )
    Sprite Rotation Mode=::NewEnumerator0
    Sprite Size=&Multiply_Vector2D(
     A=&SelectVector2DFromArray(
      Direct Array Index=&ReturnExecIndex()
      Vector 2D Selection Array=DataInterface() {propertyValues {FloatData [<500 5>, <5 500>, <200 200>, <4000 25>]}}
     )
     B=<2 2>
    )
    Sprite Size Mode=::NewEnumerator4
    UsePositionOffset=false
    Write Material Random=true
   )]
  particleUpdateScript [
   &ParticleState()
   &SpriteRotationRate(Rotation Rate=&FloatFromCurve(FloatCurve=DataInterface() {propertyValues {Curve {
        defaultValue=1
        keys [
         {arriveTangent=4.1785, interpMode=RCIM_Cubic, leaveTangent=4.1785, tangentMode=RCTM_User, time=0, value=0}
         {arriveTangent=1.1179, interpMode=RCIM_Cubic, leaveTangent=1.1179, tangentMode=RCTM_User, time=0.2042, value=0.6228}
         {arriveTangent=0.0224, interpMode=RCIM_Cubic, leaveTangent=0.0224, tangentMode=RCTM_User, time=1, value=1}
        ]
       }}}))
   &ScaleSpriteSize(
    Uniform Curve Scale=&Waveform("Waveform [1]"=::NewEnumerator8, "[1] Frequency"=45, "[1] Amplitude Min/Max"=<0.2 1>)
    Uniform Curve Sprite Scale=DataInterface() {propertyValues {Curve {
       defaultValue=0
       keys [
        {arriveTangent=1.2078, interpMode=RCIM_Cubic, time=0, value=0.5}
        {arriveTangent=0.0405, interpMode=RCIM_Cubic, leaveTangent=0.0405, leaveTangentWeight=0.2866, tangentMode=RCTM_User, time=0.2, value=0.1884}
        {
         arriveTangent=2.5249
         arriveTangentWeight=0.7778
         interpMode=RCIM_Cubic
         leaveTangent=2.5249
         leaveTangentWeight=0.1274
         tangentMode=RCTM_User
         time=0.8592
         value=0.9847
        }
        {arriveTangent=-9.0725, arriveTangentWeight=0.4283, interpMode=RCIM_Cubic, leaveTangent=-9.0725, tangentMode=RCTM_User, time=1, value=0}
       ]
      }}}
   )
   &DynamicMaterialParameters(
    Index 0 Param 4=0.001
    Module.Index0Float0Write=true
    Module.Index0Float1Write=true
    Module.Index0Float2Write=true
    Module.Index0Float3Write=true
   )
   &OffsetPosition(Position Offset=&MakeVector(
     X=0
     Y=0
     Z=&FloatFromCurve(
      FloatCurve=DataInterface() {propertyValues {Curve {
         defaultValue=0
         keys [
          {arriveTangent=4.1785, interpMode=RCIM_Cubic, leaveTangent=4.1785, tangentMode=RCTM_User, time=0, value=0}
          {arriveTangent=1.1179, interpMode=RCIM_Cubic, leaveTangent=1.1179, tangentMode=RCTM_User, time=0.2042, value=0.6228}
          {arriveTangent=0.0224, interpMode=RCIM_Cubic, leaveTangent=0.0224, tangentMode=RCTM_User, time=1, value=1}
         ]
        }}}
      Scale Curve=120
     )
    ))
  ]
  rendererClasses [&NiagaraSpriteRendererProperties]
  renderers [{rendererClass=&NiagaraSpriteRendererProperties}]
 }}
summary {
 systemName=NS_Umbramolt_Kipper_Charge
 emitters [{emitterName=Minimal, rendererClasses [&NiagaraSpriteRendererProperties]}]
 userVariables []
}
