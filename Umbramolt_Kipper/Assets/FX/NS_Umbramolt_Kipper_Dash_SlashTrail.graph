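# Niagara system for the slash trail effect of Umbramolt Kipper's dash, judging by its name.
# The summary lists no emitters, user variables or data interfaces, only system-level modules.
# Six Set Variables assignments and System State make up those modules.
# Normalize Vector and Make Quaternion dynamic inputs are also used, suggesting that a direction or orientation is computed at system level.
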NiagaraSystem /Umbramolt_Kipper/Assets/FX/NS_Umbramolt_Kipper_Dash_SlashTrail
dependencies {
 usedDataInterfaces []
 usedDynamicInputs [&NormalizeVector, &MakeQuaternion]
 usedModules [
  &self:SystemSpawnScript.SystemScriptSource.SystemScriptGraph.NiagaraNodeAssignment_0.SetVariables_#1
  &self:SystemSpawnScript.SystemScriptSource.SystemScriptGraph.NiagaraNodeAssignment_1.SetVariables_#2
  &self:SystemSpawnScript.SystemScriptSource.SystemScriptGraph.NiagaraNodeAssignment_4.SetVariables_#3
  &SystemState
  &self:SystemSpawnScript.SystemScriptSource.SystemScriptGraph.NiagaraNodeAssignment_2.SetVariables_#4
  &self:SystemSpawnScript.SystemScriptSource.SystemScriptGraph.NiagaraNodeAssignment_5.SetVariables_#5
  &self:SystemSpawnScript.SystemScriptSource.SystemScriptGraph.NiagaraNodeAssignment_6.SetVariables_#6
 ]
}
summary {systemName=NS_Umbramolt_Kipper_Dash_SlashTrail, emitters [], userVariables []}
