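# Translucent lit overlay material for the Umbramolt Kipper skeletal mesh, with mesh deformer support and a bright green color.
# An IsMaxed? parameter switches between a time-driven sine pulse, remapped to stay partly visible, and a fixed Opacity parameter.
# Only negative IsMaxed? values select the fixed value, so the default pulses.
# The selected value is squared and clamped, while a separate branch multiplies the green by a Strength parameter.
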Material /Umbramolt_Kipper/Assets/Materials/M_Umbramolt_Kipper_Overlay
graph {
 properties {
  bAutomaticallySetUsageInEditor=false
  bUsedWithMeshDeformer=true
  bUsedWithSkeletalMesh=true
  blendMode=BLEND_Translucent
  usedShadingModels=Default Lit
 }
 nodes [
  n1 = Time()
  n2 = Constant() {r=1}
  n3 = MaterialFunctionCall(Input=Sine(None=Multiply(A=@n1) {constB=2}), Input Low=Constant() {r=-1}, Input High=@n2, Target Low=Constant() {r=0.2}, Target High=@n2) {materialFunction=&RemapValueRange}
  n4 = Divide(A=If(
    A=ScalarParameter() {parameterName=IsMaxed?}
    "A > B"=@n3.Result
    "A == B"=@n3.Result
    "A < B"=ScalarParameter() {defaultValue=1, parameterName=Opacity}
   ) {constB=1})
  Clamp(None=Multiply(A=@n4, B=@n4))
  Multiply(A=Constant4Vector() {constant=<rgba 0.2468 1 0 1>}.RGBA, B=ScalarParameter() {defaultValue=20, parameterName=Strength})
 ]
}
